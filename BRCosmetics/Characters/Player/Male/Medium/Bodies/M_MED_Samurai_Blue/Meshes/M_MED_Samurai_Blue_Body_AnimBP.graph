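# Post-process style AnimBP for the Samurai Blue body cosmetic, adding secondary motion to a pose copied from the parent mesh.
# RBF pose drivers position shoulder armor, switching between grounded and skydiving sets.
# AnimDynamics chains then simulate shoulder armor, hip armor and thigh-colliding loincloths, alongside a constrained backpack mount.
# Perlin noise supplies skydiving forces; long backbling disables back-loincloth simulation, and any backbling sets a material hide parameter.
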AnimBlueprint /BRCosmetics/Characters/Player/Male/Medium/Bodies/M_MED_Samurai_Blue/Meshes/M_MED_Samurai_Blue_Body_AnimBP
target_skeleton=&/M_MED_Samurai_Blue_Body_Skeleton
graphs [
 AnimGraph {nodes [
   n1 = Get falling_external_force_z() -> (falling_external_force_z:Vector)
   n2 = Pose Driver - Source: clavicle_l!(SourcePose:Pose Link Structure=Use cached pose 'CopyPose_Cache'()) -> (Pose:Pose Link Structure) {
    is=title AnimGraphNode_PoseDriver
    poseAsset=&/M_MED_Samurai_Blue_Body_Grounded_RBF_PoseAsset
    onlyDriveBones [{boneName=dyn_shoulderArmor_master_l}, {boneName=dyn_shoulderArmor_offset_l}]
    poseTargets [
     {
      bApplyCustomCurve=false
      bIsHidden=false
      distanceMethod=DefaultMethod
      drivenName=Pose_0
      functionType=DefaultFunction
      targetScale=0.2379
      boneTransforms [
       {targetRotation=<rot -85.1915 -138.5625 143.0619>, targetTranslation=<4.3575 -0.1466 -2.3223>}
       {targetRotation=<rot -64.7854 -3.089 6.0959>, targetTranslation=<15.5111 0 0>}
      ]
      customCurve {defaultValue=3.4028234663852886e+38, postInfinityExtrap=RCCE_Constant, preInfinityExtrap=RCCE_Constant, keys []}
      targetRotation=<rot 0 0 0>
     }
     {
      bApplyCustomCurve=false
      bIsHidden=false
      distanceMethod=DefaultMethod
      drivenName=Pose_1
      functionType=DefaultFunction
      targetScale=1
      boneTransforms [
       {targetRotation=<rot -71.8075 -7.0773 -33.3059>, targetTranslation=<4.3575 -0.1466 -2.3223>}
       {targetRotation=<rot -9.5233 -21.9053 -28.615>, targetTranslation=<15.5111 0 0>}
      ]
      customCurve {defaultValue=3.4028234663852886e+38, postInfinityExtrap=RCCE_Constant, preInfinityExtrap=RCCE_Constant, keys []}
      targetRotation=<rot 0 0 0>
     }
     {
      bApplyCustomCurve=false
      bIsHidden=false
      distanceMethod=DefaultMethod
      drivenName=Pose_2
      functionType=DefaultFunction
      targetScale=0.308
      boneTransforms [
       {targetRotation=<rot -75.4617 19.5667 -12.8774>, targetTranslation=<4.3575 -0.1466 -2.3223>}
       {targetRotation=<rot -64.9089 79.0369 -47.2079>, targetTranslation=<15.5111 0 0>}
      ]
      customCurve {defaultValue=3.4028234663852886e+38, postInfinityExtrap=RCCE_Constant, preInfinityExtrap=RCCE_Constant, keys []}
      targetRotation=<rot 0 0 0>
     }
     {
      bApplyCustomCurve=false
      bIsHidden=false
      distanceMethod=DefaultMethod
      drivenName=Pose_3
      functionType=DefaultFunction
      targetScale=0.4931
      boneTransforms [
       {targetRotation=<rot -74.7073 55.0349 -49.5566>, targetTranslation=<4.3575 -0.1466 -2.3223>}
       {targetRotation=<rot -36.5188 -9.2495 27.844>, targetTranslation=<15.5111 0 0>}
      ]
      customCurve {defaultValue=3.4028234663852886e+38, postInfinityExtrap=RCCE_Constant, preInfinityExtrap=RCCE_Constant, keys []}
      targetRotation=<rot 0 0 0>
     }
     {
      bApplyCustomCurve=false
      bIsHidden=false
      distanceMethod=DefaultMethod
      drivenName=Pose_4
      functionType=DefaultFunction
      targetScale=0.2379
      boneTransforms [
       {targetRotation=<rot -81.7412 -135.4938 140.2621>, targetTranslation=<4.3575 -0.1466 -2.3223>}
       {targetRotation=<rot -49.8941 -8.1717 1.1726>, targetTranslation=<15.5111 0 0>}
      ]
      customCurve {defaultValue=3.4028234663852886e+38, postInfinityExtrap=RCCE_Constant, preInfinityExtrap=RCCE_Constant, keys []}
      targetRotation=<rot 0 0 0>
     }
     {
      bApplyCustomCurve=false
      bIsHidden=false
      distanceMethod=DefaultMethod
      drivenName=Pose_5
      functionType=DefaultFunction
      targetScale=0.7376
      boneTransforms [
       {targetRotation=<rot -80.8622 39.0398 -38.1227>, targetTranslation=<4.3575 -0.1466 -2.3223>}
       {targetRotation=<rot -31.8256 65.8251 -86.0604>, targetTranslation=<15.5111 0 0>}
      ]
      customCurve {defaultValue=3.4028234663852886e+38, postInfinityExtrap=RCCE_Constant, preInfinityExtrap=RCCE_Constant, keys []}
      targetRotation=<rot 0 0 0>
     }
     {
      bApplyCustomCurve=false
      bIsHidden=false
      distanceMethod=DefaultMethod
      drivenName=Pose_6
      functionType=DefaultFunction
      targetScale=0.558
      boneTransforms [
       {targetRotation=<rot -47.5697 14.692 -0.434>, targetTranslation=<4.3575 -0.1466 -2.3223>}
       {targetRotation=<rot -8.3098 26.4205 -78.9654>, targetTranslation=<15.5111 0 0>}
      ]
      customCurve {defaultValue=3.4028234663852886e+38, postInfinityExtrap=RCCE_Constant, preInfinityExtrap=RCCE_Constant, keys []}
      targetRotation=<rot 0 0 0>
     }
     {
      bApplyCustomCurve=false
      bIsHidden=false
      distanceMethod=DefaultMethod
      drivenName=Pose_7
      functionType=DefaultFunction
      targetScale=0.308
      boneTransforms [
       {targetRotation=<rot -86.0332 143.397 -137.7672>, targetTranslation=<4.3575 -0.1466 -2.3223>}
       {targetRotation=<rot -67.1606 51.1714 -32.7303>, targetTranslation=<15.5111 0 0>}
      ]
      customCurve {defaultValue=3.4028234663852886e+38, postInfinityExtrap=RCCE_Constant, preInfinityExtrap=RCCE_Constant, keys []}
      targetRotation=<rot 0 0 0>
     }
     {
      bApplyCustomCurve=false
      bIsHidden=false
      distanceMethod=DefaultMethod
      drivenName=Pose_8
      functionType=DefaultFunction
      targetScale=0.2698
      boneTransforms [
       {targetRotation=<rot -84.0122 51.6418 -47.6753>, targetTranslation=<4.3575 -0.1466 -2.3223>}
       {targetRotation=<rot -74.2195 -36.1046 27.6902>, targetTranslation=<15.5111 0 0>}
      ]
      customCurve {defaultValue=3.4028234663852886e+38, postInfinityExtrap=RCCE_Constant, preInfinityExtrap=RCCE_Constant, keys []}
      targetRotation=<rot 0 0 0>
     }
     {
      bApplyCustomCurve=false
      bIsHidden=false
      distanceMethod=DefaultMethod
      drivenName=Pose_9
      functionType=DefaultFunction
      targetScale=0.558
      boneTransforms [
       {targetRotation=<rot -48.3561 10.1188 -2.4428>, targetTranslation=<4.3575 -0.1466 -2.3223>}
       {targetRotation=<rot 14.5763 54.8781 -98.5188>, targetTranslation=<15.5111 0 0>}
      ]
      customCurve {defaultValue=3.4028234663852886e+38, postInfinityExtrap=RCCE_Constant, preInfinityExtrap=RCCE_Constant, keys []}
      targetRotation=<rot 0 0 0>
     }
     {
      bApplyCustomCurve=false
      bIsHidden=false
      distanceMethod=DefaultMethod
      drivenName=Pose_10
      functionType=DefaultFunction
      targetScale=0.7228
      boneTransforms [
       {targetRotation=<rot -79.2874 99.6397 -110.9321>, targetTranslation=<4.3575 -0.1466 -2.3223>}
       {targetRotation=<rot -14.9047 102.8909 -17.6919>, targetTranslation=<15.5111 0 0>}
      ]
      customCurve {defaultValue=3.4028234663852886e+38, postInfinityExtrap=RCCE_Constant, preInfinityExtrap=RCCE_Constant, keys []}
      targetRotation=<rot 0 0 0>
     }
     {
      bApplyCustomCurve=false
      bIsHidden=false
      distanceMethod=DefaultMethod
      drivenName=Pose_11
      functionType=DefaultFunction
      targetScale=0.7228
      boneTransforms [
       {targetRotation=<rot -58.4846 -9.5373 2.6759>, targetTranslation=<4.3575 -0.1466 -2.3223>}
       {targetRotation=<rot -44.7022 128.4906 -48.0679>, targetTranslation=<15.5111 0 0>}
      ]
      customCurve {defaultValue=3.4028234663852886e+38, postInfinityExtrap=RCCE_Constant, preInfinityExtrap=RCCE_Constant, keys []}
      targetRotation=<rot 0 0 0>
     }
    ]
    rBFParams {distanceMethod=Quaternion, function=Linear, radius=100}
    sourceBones [{boneName=clavicle_l}, {boneName=upperarm_l}]
   }
   n3 = Pose Driver - Source: clavicle_r!(SourcePose:Pose Link Structure=@n2) -> (Pose:Pose Link Structure) {
    is=title AnimGraphNode_PoseDriver
    poseAsset=&M_MED_Samurai_Blue_Body_Grounded_RBF_PoseAsset
    onlyDriveBones [{boneName=dyn_shoulderArmor_master_r}, {boneName=dyn_shoulderArmor_offset_r}]
    poseTargets [
     {
      bApplyCustomCurve=false
      bIsHidden=false
      distanceMethod=DefaultMethod
      drivenName=Pose_0
      functionType=DefaultFunction
      targetScale=0.2938
      boneTransforms [
       {targetRotation=<rot -85.1915 41.4375 143.0619>, targetTranslation=<4.3575 -0.1466 2.3223>}
       {targetRotation=<rot -64.7854 -3.089 6.0959>, targetTranslation=<-15.5111 0 0>}
      ]
      customCurve {defaultValue=3.4028234663852886e+38, postInfinityExtrap=RCCE_Constant, preInfinityExtrap=RCCE_Constant, keys []}
      targetRotation=<rot 0 0 0>
     }
     {
      bApplyCustomCurve=false
      bIsHidden=false
      distanceMethod=DefaultMethod
      drivenName=Pose_1
      functionType=DefaultFunction
      targetScale=0.5605
      boneTransforms [
       {targetRotation=<rot -64.2747 -120.6598 -83.241>, targetTranslation=<4.3575 -0.1466 2.3223>}
       {targetRotation=<rot -23.3709 86.7554 -36.3331>, targetTranslation=<-15.5111 0 0>}
      ]
      customCurve {defaultValue=3.4028234663852886e+38, postInfinityExtrap=RCCE_Constant, preInfinityExtrap=RCCE_Constant, keys []}
      targetRotation=<rot 0 0 0>
     }
     {
      bApplyCustomCurve=false
      bIsHidden=false
      distanceMethod=DefaultMethod
      drivenName=Pose_2
      functionType=DefaultFunction
      targetScale=0.6181
      boneTransforms [
       {targetRotation=<rot -69.838 165.5965 22.6769>, targetTranslation=<4.3575 -0.1466 2.3223>}
       {targetRotation=<rot -75.1088 43.0245 -78.222>, targetTranslation=<-15.5111 0 0>}
      ]
      customCurve {defaultValue=3.4028234663852886e+38, postInfinityExtrap=RCCE_Constant, preInfinityExtrap=RCCE_Constant, keys []}
      targetRotation=<rot 0 0 0>
     }
     {
      bApplyCustomCurve=false
      bIsHidden=false
      distanceMethod=DefaultMethod
      drivenName=Pose_3
      functionType=DefaultFunction
      targetScale=0.3174
      boneTransforms [
       {targetRotation=<rot -73.4134 -176.9703 4.3592>, targetTranslation=<4.3575 -0.1466 2.3223>}
       {targetRotation=<rot -75.5657 105.7208 -93.8077>, targetTranslation=<-15.5111 0 0>}
      ]
      customCurve {defaultValue=3.4028234663852886e+38, postInfinityExtrap=RCCE_Constant, preInfinityExtrap=RCCE_Constant, keys []}
      targetRotation=<rot 0 0 0>
     }
     {
      bApplyCustomCurve=false
      bIsHidden=false
      distanceMethod=DefaultMethod
      drivenName=Pose_4
      functionType=DefaultFunction
      targetScale=0.2938
      boneTransforms [
       {targetRotation=<rot -87.7995 76.2682 109.0761>, targetTranslation=<4.3575 -0.1466 2.3223>}
       {targetRotation=<rot -69.7589 51.1037 -35.2693>, targetTranslation=<-15.5111 0 0>}
      ]
      customCurve {defaultValue=3.4028234663852886e+38, postInfinityExtrap=RCCE_Constant, preInfinityExtrap=RCCE_Constant, keys []}
      targetRotation=<rot 0 0 0>
     }
     {
      bApplyCustomCurve=false
      bIsHidden=false
      distanceMethod=DefaultMethod
      drivenName=Pose_5
      functionType=DefaultFunction
      targetScale=0.2327
      boneTransforms [
       {targetRotation=<rot -76.2754 172.1506 9.5506>, targetTranslation=<4.3575 -0.1466 2.3223>}
       {targetRotation=<rot -45.0185 104.3596 -56.1379>, targetTranslation=<-15.5111 0 0>}
      ]
      customCurve {defaultValue=3.4028234663852886e+38, postInfinityExtrap=RCCE_Constant, preInfinityExtrap=RCCE_Constant, keys []}
      targetRotation=<rot 0 0 0>
     }
     {
      bApplyCustomCurve=false
      bIsHidden=false
      distanceMethod=DefaultMethod
      drivenName=Pose_6
      functionType=DefaultFunction
      targetScale=0.2327
      boneTransforms [
       {targetRotation=<rot -74.8686 -132.7057 -46.9771>, targetTranslation=<4.3575 -0.1466 2.3223>}
       {targetRotation=<rot -39.6825 113.2505 -70.71>, targetTranslation=<-15.5111 0 0>}
      ]
      customCurve {defaultValue=3.4028234663852886e+38, postInfinityExtrap=RCCE_Constant, preInfinityExtrap=RCCE_Constant, keys []}
      targetRotation=<rot 0 0 0>
     }
     {
      bApplyCustomCurve=false
      bIsHidden=false
      distanceMethod=DefaultMethod
      drivenName=Pose_7
      functionType=DefaultFunction
      targetScale=0.4039
      boneTransforms [
       {targetRotation=<rot -57.8419 -161.5812 -18.7948>, targetTranslation=<4.3575 -0.1466 2.3223>}
       {targetRotation=<rot -66.8684 120.4314 -71.1722>, targetTranslation=<-15.5111 0 0>}
      ]
      customCurve {defaultValue=3.4028234663852886e+38, postInfinityExtrap=RCCE_Constant, preInfinityExtrap=RCCE_Constant, keys []}
      targetRotation=<rot 0 0 0>
     }
     {
      bApplyCustomCurve=false
      bIsHidden=false
      distanceMethod=DefaultMethod
      drivenName=Pose_8
      functionType=DefaultFunction
      targetScale=0.6101
      boneTransforms [
       {targetRotation=<rot -53.6465 -152.7178 -29.1383>, targetTranslation=<4.3575 -0.1466 2.3223>}
       {targetRotation=<rot -17.8084 26.8504 -97.253>, targetTranslation=<-15.5111 0 0>}
      ]
      customCurve {defaultValue=3.4028234663852886e+38, postInfinityExtrap=RCCE_Constant, preInfinityExtrap=RCCE_Constant, keys []}
      targetRotation=<rot 0 0 0>
     }
     {
      bApplyCustomCurve=false
      bIsHidden=false
      distanceMethod=DefaultMethod
      drivenName=Pose_9
      functionType=DefaultFunction
      targetScale=1
      boneTransforms [
       {targetRotation=<rot -68.9933 -127.6298 -53.9108>, targetTranslation=<4.3575 -0.1466 2.3223>}
       {targetRotation=<rot 33.2732 89.3676 -89.1674>, targetTranslation=<-15.5111 0 0>}
      ]
      customCurve {defaultValue=3.4028234663852886e+38, postInfinityExtrap=RCCE_Constant, preInfinityExtrap=RCCE_Constant, keys []}
      targetRotation=<rot 0 0 0>
     }
     {
      bApplyCustomCurve=false
      bIsHidden=false
      distanceMethod=DefaultMethod
      drivenName=Pose_10
      functionType=DefaultFunction
      targetScale=0.6101
      boneTransforms [
       {targetRotation=<rot -83.6219 69.4343 115.2173>, targetTranslation=<4.3575 -0.1466 2.3223>}
       {targetRotation=<rot -11.4935 30.1498 -76.7725>, targetTranslation=<-15.5111 0 0>}
      ]
      customCurve {defaultValue=3.4028234663852886e+38, postInfinityExtrap=RCCE_Constant, preInfinityExtrap=RCCE_Constant, keys []}
      targetRotation=<rot 0 0 0>
     }
     {
      bApplyCustomCurve=false
      bIsHidden=false
      distanceMethod=DefaultMethod
      drivenName=Pose_11
      functionType=DefaultFunction
      targetScale=0.571
      boneTransforms [
       {targetRotation=<rot -47.9536 -145.5199 -31.7961>, targetTranslation=<4.3575 -0.1466 2.3223>}
       {targetRotation=<rot -20.6591 112.0123 -37.7437>, targetTranslation=<-15.5111 0 0>}
      ]
      customCurve {defaultValue=3.4028234663852886e+38, postInfinityExtrap=RCCE_Constant, preInfinityExtrap=RCCE_Constant, keys []}
      targetRotation=<rot 0 0 0>
     }
    ]
    rBFParams {distanceMethod=Quaternion, radius=100}
    sourceBones [{boneName=clavicle_r}, {boneName=upperarm_r}]
   }
   n4 = Pose Driver - Source: upperarm_l!(SourcePose:Pose Link Structure=Use cached pose 'CopyPose_Cache'()) -> (Pose:Pose Link Structure) {
    is=title AnimGraphNode_PoseDriver
    poseAsset=&/M_MED_Samurai_Blue_Body_Skydiving_RBF_PoseAsset
    onlyDriveBones [{boneName=dyn_shoulderArmor_master_l}, {boneName=dyn_shoulderArmor_offset_l}]
    poseTargets [
     {
      bApplyCustomCurve=false
      bIsHidden=false
      distanceMethod=DefaultMethod
      drivenName=Pose_12
      functionType=DefaultFunction
      targetScale=1
      boneTransforms [{targetRotation=<rot -64.7854 -3.089 6.0959>, targetTranslation=<15.5111 0 0>}]
      customCurve {defaultValue=3.4028234663852886e+38, postInfinityExtrap=RCCE_Constant, preInfinityExtrap=RCCE_Constant, keys []}
      targetRotation=<rot 0 0 0>
     }
     {
      bApplyCustomCurve=false
      bIsHidden=false
      distanceMethod=DefaultMethod
      drivenName=Pose_13
      functionType=DefaultFunction
      targetScale=0.3618
      boneTransforms [{targetRotation=<rot -17.9906 -27.503 -67.4976>, targetTranslation=<15.5111 0 0>}]
      customCurve {defaultValue=3.4028234663852886e+38, postInfinityExtrap=RCCE_Constant, preInfinityExtrap=RCCE_Constant, keys []}
      targetRotation=<rot 0 0 0>
     }
     {
      bApplyCustomCurve=false
      bIsHidden=false
      distanceMethod=DefaultMethod
      drivenName=Pose_14
      functionType=DefaultFunction
      targetScale=0.3618
      boneTransforms [{targetRotation=<rot 17.157 -17.077 -68.3926>, targetTranslation=<15.5111 0 0>}]
      customCurve {defaultValue=3.4028234663852886e+38, postInfinityExtrap=RCCE_Constant, preInfinityExtrap=RCCE_Constant, keys []}
      targetRotation=<rot 0 0 0>
     }
    ]
    rBFParams {distanceMethod=Quaternion, radius=100}
    sourceBones [{boneName=upperarm_l}]
   }
   n5 = Pose Driver - Source: upperarm_r!(SourcePose:Pose Link Structure=@n4) -> (Pose:Pose Link Structure) {
    is=title AnimGraphNode_PoseDriver
    poseAsset=&M_MED_Samurai_Blue_Body_Skydiving_RBF_PoseAsset
    onlyDriveBones [{boneName=dyn_shoulderArmor_master_r}, {boneName=dyn_shoulderArmor_offset_r}]
    poseTargets [
     {
      bApplyCustomCurve=false
      bIsHidden=false
      distanceMethod=DefaultMethod
      drivenName=Pose_12
      functionType=DefaultFunction
      targetScale=1
      boneTransforms [{targetRotation=<rot -64.7854 -3.089 6.0959>, targetTranslation=<-15.5111 0 0>}]
      customCurve {defaultValue=3.4028234663852886e+38, postInfinityExtrap=RCCE_Constant, preInfinityExtrap=RCCE_Constant, keys []}
      targetRotation=<rot 0 0 0>
     }
     {
      bApplyCustomCurve=false
      bIsHidden=false
      distanceMethod=DefaultMethod
      drivenName=Pose_13
      functionType=DefaultFunction
      targetScale=0.3887
      boneTransforms [{targetRotation=<rot -22.7967 -31.6166 -51.4767>, targetTranslation=<-15.5111 0 0>}]
      customCurve {defaultValue=3.4028234663852886e+38, postInfinityExtrap=RCCE_Constant, preInfinityExtrap=RCCE_Constant, keys []}
      targetRotation=<rot 0 0 0>
     }
     {
      bApplyCustomCurve=false
      bIsHidden=false
      distanceMethod=DefaultMethod
      drivenName=Pose_14
      functionType=DefaultFunction
      targetScale=0.3887
      boneTransforms [{targetRotation=<rot 1.7785 -12.1856 -70.9632>, targetTranslation=<-15.5111 0 0>}]
      customCurve {defaultValue=3.4028234663852886e+38, postInfinityExtrap=RCCE_Constant, preInfinityExtrap=RCCE_Constant, keys []}
      targetRotation=<rot 0 0 0>
     }
    ]
    rBFParams {distanceMethod=Quaternion, radius=100}
    sourceBones [{boneName=upperarm_r}]
   }
   n6 = Get bIsSkyDiving()
   n7 = Get PawnSpeed2D()
   n8 = Anim Dynamics - Bone: dyn_shoulderArmor_l(
    ComponentPose=Local To Component(LocalPose=Blend Poses by bool(BlendPose_0=@n5, BlendPose_1=@n3, bActiveValue=@n6) {blendType=Linear})
    Alpha=@n7
    ExternalForce="0,0,0"
   ) {
    bGravityOverrideInSimSpace=true
    lODThreshold=1
    alphaBoolBlend {blendOption=Linear}
    alphaScaleBiasClamp {bClampResult=true, clampMin=0.25, scale=0.01}
    boundBone {boneName=dyn_shoulderArmor_l}
    physicsBodyDefinitions [{
      sphereCollisionRadius=0
      boundBone {boneName=dyn_shoulderArmor_l}
      boxExtents=<5 5 5>
      constraintSetup {angularLimitsMax=<30 15 0>, angularLimitsMin=<-30 -15 0>, angularTarget=<0 0 0>, linearAxesMax=<0.25 0 0.25>, linearAxesMin=<-0.25 0 -0.25>}
      localJointOffset=<0.5 0 0>
     }]
   }
   n9 = Anim Dynamics - Bone: dyn_shoulderArmor_r(ComponentPose=@n8, Alpha=@n7, ExternalForce="0,0,0") {
    bGravityOverrideInSimSpace=true
    lODThreshold=1
    alphaBoolBlend {blendOption=Linear}
    alphaScaleBiasClamp {bClampResult=true, clampMin=0.25, scale=0.01}
    boundBone {boneName=dyn_shoulderArmor_r}
    physicsBodyDefinitions [{
      sphereCollisionRadius=0
      boundBone {boneName=dyn_shoulderArmor_r}
      boxExtents=<5 5 5>
      constraintSetup {angularLimitsMax=<30 15 0>, angularLimitsMin=<-30 -15 0>, angularTarget=<0 0 0>, linearAxesMax=<0.25 0 0.25>, linearAxesMin=<-0.25 0 -0.25>}
      localJointOffset=<0.5 0 0>
     }]
   }
   n10 = Get falling_external_force_y() -> (falling_external_force_y:Vector)
   n11 = Anim Dynamics - Bone: dyn_hip_armor_02_l(ComponentPose=@n9, Alpha=1, ExternalForce=@n10) {
    angularDampingOverride=0.99
    angularSpringConstant=250
    bAngularSpring=true
    bGravityOverrideInSimSpace=true
    bOverrideAngularDamping=true
    gravityScale=0.9
    numSolverIterationsPostUpdate=4
    numSolverIterationsPreUpdate=16
    alphaBoolBlend {blendOption=Linear}
    boundBone {boneName=dyn_hip_armor_02_l}
    chainEnd {boneName=dyn_hip_armor_03_l}
    physicsBodyDefinitions [{
      sphereCollisionRadius=0
      boundBone {boneName=dyn_hip_armor_02_l}
      boxExtents=<2 10 10>
      constraintSetup {angularLimitsMax=<0 0 0>, angularLimitsMin=<0 0 -30>, angularTarget=<0 0 0>}
      localJointOffset=<2 0 0>
     }]
   }
   n12 = Anim Dynamics - Bone: dyn_hip_armor_03_l(ComponentPose=@n11, Alpha=1, ExternalForce=@n10) {
    angularDampingOverride=0.99
    angularSpringConstant=250
    bAngularSpring=true
    bGravityOverrideInSimSpace=true
    bOverrideAngularDamping=true
    gravityScale=0.9
    numSolverIterationsPostUpdate=4
    numSolverIterationsPreUpdate=16
    alphaBoolBlend {blendOption=Linear}
    boundBone {boneName=dyn_hip_armor_03_l}
    chainEnd {boneName=dyn_hip_armor_03_l}
    physicsBodyDefinitions [{
      sphereCollisionRadius=0
      boundBone {boneName=dyn_hip_armor_03_l}
      boxExtents=<2 10 10>
      constraintSetup {angularLimitsMax=<0 0 0>, angularLimitsMin=<0 0 -30>, angularTarget=<0 0 0>}
      localJointOffset=<2 0 0>
     }]
   }
   n13 = Anim Dynamics - Bone: dyn_hip_armor_02_r(ComponentPose=@n12, Alpha=1, ExternalForce=@n10) {
    angularDampingOverride=0.99
    angularSpringConstant=250
    bAngularSpring=true
    bGravityOverrideInSimSpace=true
    bOverrideAngularDamping=true
    gravityScale=0.9
    numSolverIterationsPostUpdate=4
    numSolverIterationsPreUpdate=16
    alphaBoolBlend {blendOption=Linear}
    boundBone {boneName=dyn_hip_armor_02_r}
    chainEnd {boneName=dyn_hip_armor_03_r}
    physicsBodyDefinitions [{
      sphereCollisionRadius=0
      boundBone {boneName=dyn_hip_armor_02_r}
      boxExtents=<2 10 10>
      constraintSetup {angularLimitsMax=<0 0 30>, angularLimitsMin=<0 0 0>, angularTarget=<0 0 0>}
      localJointOffset=<0.1 0 0>
     }]
   }
   n14 = Anim Dynamics - Bone: dyn_hip_armor_03_r(ComponentPose=@n13, Alpha=1, ExternalForce=@n10) {
    angularDampingOverride=0.99
    angularSpringConstant=250
    bAngularSpring=true
    bGravityOverrideInSimSpace=true
    bOverrideAngularDamping=true
    gravityScale=0.9
    numSolverIterationsPostUpdate=4
    numSolverIterationsPreUpdate=16
    alphaBoolBlend {blendOption=Linear}
    boundBone {boneName=dyn_hip_armor_03_r}
    chainEnd {boneName=dyn_hip_armor_03_r}
    physicsBodyDefinitions [{
      sphereCollisionRadius=0
      boundBone {boneName=dyn_hip_armor_03_r}
      boxExtents=<2 10 10>
      constraintSetup {angularLimitsMax=<0 0 30>, angularLimitsMin=<0 0 0>, angularTarget=<0 0 0>}
      localJointOffset=<0.1 0 0>
     }]
   }
   n15 = "Constraint - dyn_backpack_mount (PPPPP)"!(
    ComponentPose=@n14
    Alpha=1
    ConstraintWeights_0=0.235
    ConstraintWeights_1:float=0.3342
    ConstraintWeights_2:float=0.1285
    ConstraintWeights_3:float=0.0645
    ConstraintWeights_4:float=0.2379
   ) -> (Pose) {
    alphaBoolBlend {blendOption=Linear}
    boneToModify {boneName=dyn_backpack_mount}
    constraintSetup [
     {targetBone {boneName=clavicle_r}}
     {targetBone {boneName=spine_03}}
     {targetBone {boneName=spine_04}}
     {targetBone {boneName=spine_05}}
     {targetBone {boneName=clavicle_l}}
    ]
    constraintWeights [1,1,1,1,1]
   }
   n16 = SelectFloat(A=0, B=SelectFloat(A=0.5, B=1, bPickA=Get bIsSkyDiving()), bPickA=Get isWearingCape() -> (isWearingCape:bool))
   n17 = Get falling_external_force_z()
   n18 = Anim Dynamics - Bone: dyn_back_loincloth_01(ComponentPose=@n15, Alpha=@n16, ExternalForce=@n17) {
    angularDampingOverride=0.999
    angularSpringConstant=250
    bAngularSpring=true
    bGravityOverrideInSimSpace=true
    bOverrideAngularDamping=true
    bUsePlanarLimit=false
    bUseSphericalLimits=true
    gravityScale=0.5
    alphaBoolBlend {blendOption=Linear}
    boundBone {boneName=dyn_back_loincloth_01}
    physicsBodyDefinitions [{
      collisionType=CustomSphere
      sphereCollisionRadius=6
      boundBone {boneName=dyn_back_loincloth_01}
      boxExtents=<5 20 20>
      constraintSetup {angularLimitsMax=<0 30 0>, angularLimitsMin=<0 -5 0>, angularTarget=<0 0 0>}
      localJointOffset=<15 0 4>
     }]
    sphericalLimits [
     {limitRadius=10, drivingBone {boneName=thigh_l}, sphereLocalOffset=<-5 0 0>}
     {limitRadius=10, drivingBone {boneName=thigh_r}, sphereLocalOffset=<5 0 0>}
     {limitRadius=5, drivingBone {boneName=pelvis}, sphereLocalOffset=<-2 0 0>}
    ]
   }
   n19 = Anim Dynamics - Bone: dyn_back_loincloth_02(ComponentPose=@n18, Alpha=@n16, ExternalForce=@n17) {
    angularDampingOverride=0.999
    angularSpringConstant=250
    bAngularSpring=true
    bGravityOverrideInSimSpace=true
    bOverrideAngularDamping=true
    bUseSphericalLimits=true
    gravityScale=0.5
    alphaBoolBlend {blendOption=Linear}
    boundBone {boneName=dyn_back_loincloth_02}
    physicsBodyDefinitions [{
      collisionType=CustomSphere
      sphereCollisionRadius=6
      boundBone {boneName=dyn_back_loincloth_02}
      boxExtents=<5 20 20>
      constraintSetup {angularLimitsMax=<0 30 0>, angularLimitsMin=<0 -5 0>, angularTarget=<0 0 0>}
      localJointOffset=<15 0 4>
     }]
    sphericalLimits [
     {limitRadius=10, drivingBone {boneName=thigh_l}, sphereLocalOffset=<-17 0 0>}
     {limitRadius=10, drivingBone {boneName=thigh_r}, sphereLocalOffset=<17 0 0>}
    ]
   }
   n20 = Anim Dynamics - Bone: dyn_back_loincloth_03(ComponentPose=@n19, Alpha=@n16, ExternalForce=@n17) {
    angularBiasOverride=0.8
    angularDampingOverride=0.999
    angularSpringConstant=250
    bAngularSpring=true
    bGravityOverrideInSimSpace=true
    bOverrideAngularBias=true
    bOverrideAngularDamping=true
    bUseSphericalLimits=true
    gravityScale=0.5
    alphaBoolBlend {blendOption=Linear}
    boundBone {boneName=dyn_back_loincloth_03}
    physicsBodyDefinitions [{
      collisionType=CustomSphere
      sphereCollisionRadius=6
      boundBone {boneName=dyn_back_loincloth_03}
      boxExtents=<5 20 20>
      constraintSetup {angularLimitsMax=<0 30 0>, angularLimitsMin=<0 -5 0>, angularTarget=<0 0 0>}
      localJointOffset=<15 0 4>
     }]
    sphericalLimits [
     {limitRadius=10, drivingBone {boneName=thigh_l}, sphereLocalOffset=<-25 0 0>}
     {limitRadius=10, drivingBone {boneName=thigh_r}, sphereLocalOffset=<25 0 0>}
    ]
   }
   n21 = Anim Dynamics - Bone: dyn_front_loincloth_01(ComponentPose=@n20, Alpha=1, ExternalForce=@n1) {
    angularDampingOverride=0.999
    angularSpringConstant=250
    bAngularSpring=true
    bGravityOverrideInSimSpace=true
    bOverrideAngularDamping=true
    bUsePlanarLimit=false
    bUseSphericalLimits=true
    gravityScale=0.5
    alphaBoolBlend {blendOption=Linear}
    boundBone {boneName=dyn_front_loincloth_01}
    physicsBodyDefinitions [{
      collisionType=CustomSphere
      sphereCollisionRadius=6
      boundBone {boneName=dyn_front_loincloth_01}
      boxExtents=<5 20 20>
      constraintSetup {angularLimitsMax=<0 5 0>, angularLimitsMin=<0 -30 0>, angularTarget=<0 0 0>}
      localJointOffset=<15 0 -4>
     }]
    sphericalLimits [
     {limitRadius=10, drivingBone {boneName=thigh_l}, sphereLocalOffset=<-5 0 0>}
     {limitRadius=10, drivingBone {boneName=thigh_r}, sphereLocalOffset=<5 0 0>}
     {limitRadius=5, drivingBone {boneName=pelvis}, sphereLocalOffset=<-2 0 0>}
    ]
   }
   n22 = Anim Dynamics - Bone: dyn_front_loincloth_02(ComponentPose=@n21, Alpha=1, ExternalForce=@n1) {
    angularDampingOverride=0.999
    angularSpringConstant=250
    bAngularSpring=true
    bGravityOverrideInSimSpace=true
    bOverrideAngularDamping=true
    bUseSphericalLimits=true
    gravityScale=0.5
    alphaBoolBlend {blendOption=Linear}
    boundBone {boneName=dyn_front_loincloth_02}
    physicsBodyDefinitions [{
      collisionType=CustomSphere
      sphereCollisionRadius=6
      boundBone {boneName=dyn_front_loincloth_02}
      boxExtents=<5 20 20>
      constraintSetup {angularLimitsMax=<0 5 0>, angularLimitsMin=<0 -30 0>, angularTarget=<0 0 0>}
      localJointOffset=<15 0 -4>
     }]
    sphericalLimits [
     {limitRadius=10, drivingBone {boneName=thigh_l}, sphereLocalOffset=<-17 0 0>}
     {limitRadius=10, drivingBone {boneName=thigh_r}, sphereLocalOffset=<17 0 0>}
    ]
   }
   n23 = Anim Dynamics - Bone: dyn_front_loincloth_03(ComponentPose=@n22, Alpha=1, ExternalForce=@n1) {
    angularBiasOverride=0.8
    angularDampingOverride=0.999
    angularSpringConstant=250
    bAngularSpring=true
    bGravityOverrideInSimSpace=true
    bOverrideAngularBias=true
    bOverrideAngularDamping=true
    bUseSphericalLimits=true
    gravityScale=0.5
    alphaBoolBlend {blendOption=Linear}
    boundBone {boneName=dyn_front_loincloth_03}
    physicsBodyDefinitions [{
      collisionType=CustomSphere
      sphereCollisionRadius=6
      boundBone {boneName=dyn_front_loincloth_03}
      boxExtents=<5 20 20>
      constraintSetup {angularLimitsMax=<0 5 0>, angularLimitsMin=<0 -30 0>, angularTarget=<0 0 0>}
      localJointOffset=<15 0 -4>
     }]
    sphericalLimits [
     {limitRadius=10, drivingBone {boneName=thigh_l}, sphereLocalOffset=<-25 0 0>}
     {limitRadius=10, drivingBone {boneName=thigh_r}, sphereLocalOffset=<25 0 0>}
    ]
   }
   Output Pose(Result=Component To Local(ComponentPose=@n23)) {layerGroup=None}
   Save cached pose 'CopyPose_Cache'(Pose=Copy Pose From Mesh() {bUseAttachedParent=true}) {cacheName=CopyPose_Cache}
   Sequence Player 'skydive_rom_test'() {sequence=&/skydive_rom_test}
  ]}
 EventGraph {nodes [
   Event OnCharacterCustomizationLoaded()
   n1 = CheckForBackbling()
   > Set isWearingBackbling(isWearingBackbling=@n1)
   > Set isWearingCape(isWearingCape=@n1.bIsBackblingLong)
   Event BlueprintUpdateAnimation()
   n2 = Get PawnLastLocation() -> (PawnLastLocation:Vector)
   n3 = SelectFloat(
    A=Make Perlin Noise and Remap!(Value:float=BreakVector(InVec=@n2).Z, RangeOutMin:float=-5000, RangeOutMax:float=5000) -> (ReturnValue:float)
    B=0
    bPickA=Get bIsSkyDiving()
   )
   n4 = SelectFloat(A=float * float(A=@n3, B=0.1), B=@n3, bPickA=Get bIsParachuteOpened())
   > Set falling_external_force_z(falling_external_force_z:Vector=MakeVector(X=@n4, Y=0, Z=@n4)) -> (Output_Get:Vector)
   > Set falling_external_force_y(falling_external_force_y:Vector=MakeVector(X=@n4, Y=@n4)) -> (Output_Get:Vector)
   > Branch(Condition=Get isWearingBackbling())
   > n5 = Get Dynamic Material Instance(SkeletalMeshComponent=GetPartSkeletalMeshComponent())
   > SetScalarParameterValue(self=@n5, ParameterName=Hide Element 02, Value=1)
   TryGetPawnOwner()
  ]}
]
variables [
 falling_external_force_y {type=Vector}
 falling_external_force_z {type=Vector}
 isWearingBackbling {type=boolean}
 isWearingCape {type=boolean}
]
