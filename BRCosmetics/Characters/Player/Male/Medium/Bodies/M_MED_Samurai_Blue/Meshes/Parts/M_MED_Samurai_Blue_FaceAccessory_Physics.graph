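# Physics asset for the face accessory mesh of the blue samurai player body (male, medium), set up for its ponytail.
# Three small capsule bodies on the dyn_ponytail bones form a chain hanging from a large zero-length capsule, effectively a sphere, on the head bone.
# Each segment is constrained to the one before it, with the first tied to the head and twist locked on every joint, so any bending comes from swing.
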PhysicsAsset /BRCosmetics/Characters/Player/Male/Medium/Bodies/M_MED_Samurai_Blue/Meshes/Parts/M_MED_Samurai_Blue_FaceAccessory_Physics
bodies {
 dyn_ponytail_01 [{shapeType=Capsule, length=2.75, radius=3.755, center=<2.6221 0 0>, rotation=<rot 90 -0.0049 -0.0049>}]
 dyn_ponytail_02 [{shapeType=Capsule, length=5.0482, radius=3.7919, center=<3.9526 0 -0.0003>, rotation=<rot 90 2.5305 2.5305>}]
 dyn_ponytail_03 [{shapeType=Capsule, length=2.75, radius=3.505, center=<3.2292 0 0>, rotation=<rot -90 -0.0198 0.0198>}]
 head [{shapeType=Capsule, length=0, radius=16.3517}]
}
constraints [
 {bone1Name=dyn_ponytail_01, bone2Name=head, swing1LimitDegrees=45, swing2LimitDegrees=45, twistLimitDegrees=45, twistMotion=Locked}
 {bone1Name=dyn_ponytail_02, bone2Name=dyn_ponytail_01, swing1LimitDegrees=45, swing2LimitDegrees=45, twistLimitDegrees=45, twistMotion=Locked}
 {bone1Name=dyn_ponytail_03, bone2Name=dyn_ponytail_02, swing1LimitDegrees=45, swing2LimitDegrees=45, twistLimitDegrees=45, twistMotion=Locked}
]
